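; Blueprint for the TeaCake dynamite consumable's weapon actor, centered on describing the shape of its fuse.
; SplinePointLocations returns four fuse socket positions from the weapon mesh, ordered FX_FuseBegin, dyn_fuse2, dyn_fuse1, FX_FuseEnd, converted into the Spline component's local space.
; The holster, audio parameter and sound indicator functions return only default values, and the construction script just calls its parent.
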(fn CanApplyHolster
  (params (HolsterId :type Name%20%28by%20ref%29 :default "None"))
  (return false))

(fn GetActorAudioParameters
  (return 0))

(fn ShouldShowSoundIndicator
  (params (PlayerController :type Fort%20Player%20Controller%20Object%20Reference))
  (return false))

(fn SplinePointLocations
  (return (Utilities|Array|MakeArray (Math|Transform|InverseTransformLocation (Transformation|GetWorldTransform (Variables|Components|GetSpline)) (Transformation|GetSocketLocation (Visual|GetWeaponMesh) "FX_FuseBegin")) (Math|Transform|InverseTransformLocation (Transformation|GetWorldTransform (Variables|Components|GetSpline)) (Transformation|GetSocketLocation (Visual|GetWeaponMesh) "dyn_fuse2")) (Math|Transform|InverseTransformLocation (Transformation|GetWorldTransform (Variables|Components|GetSpline)) (Transformation|GetSocketLocation (Visual|GetWeaponMesh) "dyn_fuse1")) (Math|Transform|InverseTransformLocation (Transformation|GetWorldTransform (Variables|Components|GetSpline)) (Transformation|GetSocketLocation (Visual|GetWeaponMesh) "FX_FuseEnd")))))

(fn ConstructionScript
  (|CallParentFunction))
