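; Dynamite grenade projectile that explodes when its fuse ends, then relaunches in bounces, exploding until a tunable bounce limit kills it.
; Tunables start the fuse on landing, first bounce or water/special-surface contact; replicated detonation time drives its timer and looping cue.
; Authority counts bounces and randomizes velocity around the first impact; clients play surface-aware FX, audio and shake.
; It also levels the bundle on stop and resumes simulation when its supporting building or vehicle is destroyed.
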(event Custom|LoopingIndicator
  ;[1 1 1] Setup for Change to Particle System if needed
  (Fortnite|ShowSoundIndicatorLocal :TrackedActor self :Location (Transformation|GetActorLocation) :MaxAudibleDistance (Variables|/Script/Engine.SoundBase|GetMaxDistance (Variables|Sounds|GetCue_GrenadeFuseSound)) :IndicatorType "Glider" :Tint "(R=1.000000,G=1.000000,B=1.000000,A=1.000000)" :OverrideIcon (Variables|Default|GetSoundIndicatorIcon) :AttenuationMaxRangeMultiplier 1.0 :AttenuationIntensityMultiplier 1.0) ; Indicator for fuse
)

(event EventBeginPlay
  (Utilities|FlowControl|Sequence
    (:then_0
      ;[1 1 1] Setup for Change to Particle System if needed
      (Variables|VFX|SetP_Explosion (select (Variables|Default|GetUse%20Shockwave%20Based%20Explosion) 0 0))
      (CallFunction|AysncLoad)
      (Networking|CanExecuteCosmeticEvents
        (:True
          (Audio|Components|Audio|FadeIn (Variables|B_Prj_Grenade|GetGrenadeFuse_AudioComponent) 0.25 1.0 0.0 "Linear" :!class /Script/Engine.AudioComponent)
          (if (Variables|Default|GetbShowFuseSoundIndicator)
            (bind _returnvalue (Utilities|Time|SetTimerbyEvent (delegate Custom|LoopingIndicator) 0.3 true))
            (Variables|Default|SetFuseIndicatorTimer _returnvalue)))))
    (:then_1
      ;[0 0.1971 0.9062 0.5] Hotfix values for bounce after fuse explode
      ; Storing this off so we don't have to convert to an Int every time we do the check
      (Variables|Default|SetNumberOfBouncesTillExplode (Math|Float|Truncate (ScalableFloat|GetValueatLevel (Variables|Default|GetHF_BouncesBeforeExplode)))))))

(event EventOnStop
  (params (Hit :type Hit%20Result%20Structure%20%28by%20ref%29))
  (bind (_bblockinghit _binitialoverlap _time _distance _location _impactpoint _normal) (Collision|BreakHitResult Hit))
  ;[1 1 1] Keeps Fuse Visible when projectile hits ground
  (|CallParentFunction Hit)
  (if (> (ScalableFloat|GetCachedScalableFloatValue (Variables|Default|GetHF_StartFuseOnLanded)) 0)
    (Utilities|FlowControl|Sequence
      (:then_0
        _)
      (:then_1
        (CallFunction|EventStartFuse (+ (Math|Float|Clamp(Float) (ScalableFloat|GetCachedScalableFloatValue (Variables|Default|GetHF_FuseDuration)) (Variables|Default|GetMinFuseTime) (ScalableFloat|GetCachedScalableFloatValue (Variables|Default|GetHF_FuseDuration))) (Fortnite|Sync|Time|GetServerWorldTimeSeconds))))))
  (Networking|CanExecuteCosmeticEvents
    (:True
      (bind _outrotation (CallFunction|MakeOrthogonalZ (Transformation|GetWorldRotation (Variables|Default|GetTNTBundle)) _normal))
      (Variables|Default|SetTargetRotation _outrotation)
      (bind _returnvalue (Utilities|Time|SetTimerbyEvent (delegate Custom|Lerp) 0.0333 true))
      (Variables|Default|SetLerpTimer _returnvalue)
      (Utilities|FlowControl|Delay (Variables|Default|GetLerpDuration))
      (Utilities|Time|ClearandInvalidateTimerbyHandle _returnvalue))))

(event OnProjectileStop(ProjectileMovementComponent)
  (params (ImpactResult :type Hit%20Result%20Structure%20%28by%20ref%29))
  (bind (_bblockinghit _binitialoverlap _time _distance _location _impactpoint _normal _impactnormal _physmat _hitactor) (Collision|BreakHitResult ImpactResult))
  ;[1 1 1] Bind events for resimulating after the attached building actor or vehicle is destroyed
  (bind _asbuilding_actor (Utilities|Casting|CastToBuildingActor _hitactor)
    (:then
      (EventDispatchers|BindEventtoOnDied _asbuilding_actor (delegate Custom|Event_BuildingActorDied) :!class /Script/FortniteGame.BuildingActor))
    (:CastFailed
      (bind _asfort_vehicle_interface (Utilities|Casting|CastToFortVehicleInterface _hitactor))
      (Game|BindEventtoOnDestroyed (Vehicle|GetActor _asfort_vehicle_interface) (delegate Custom|Event_OnVehicleDestroyed)))))

(event Custom|Lerp
  (bind _tntbundle (Variables|Default|GetTNTBundle))
  (bind _returnvalue (Transformation|GetWorldRotation _tntbundle))
  (bind _targetrotation (Variables|Default|GetTargetRotation))
  ;[1 1 1] Keeps Fuse Visible when projectile hits ground
  (if (Math|Rotator|Equal(Rotator) _returnvalue _targetrotation 0.01)
    (Utilities|Time|ClearandInvalidateTimerbyHandle (Variables|Default|GetLerpTimer))
    (else
      (Transformation|SetWorldRotation (Variables|Default|GetTNTBundle) (Math|Interpolation|RInterpTo _returnvalue _targetrotation (Utilities|Time|GetWorldDeltaSeconds) (Variables|Default|GetInterpSpeed)) false false))))

(event Custom|Event_BuildingActorDied
  (params (DamagedActor :type Actor%20Object%20Reference) (Damage :type Float%20%28single-precision%29) (InstigatedBy :type Controller%20Object%20Reference) (DamageCauser :type Actor%20Object%20Reference) (HitLocation :type Vector) (FHitComponent :type Primitive%20Component%20Object%20Reference) (BoneName :type Name) (Momentum :type Vector))
  ;[1 1 1] Bind events for resimulating after the attached building actor or vehicle is destroyed
  (Projectile|ResumeSimulation))

(event Custom|Event_OnVehicleDestroyed
  (params (DestroyedActor :type Actor%20Object%20Reference))
  ;[1 1 1] Bind events for resimulating after the attached building actor or vehicle is destroyed
  (Projectile|ResumeSimulation))

(event EventDestroyed
  ;[1 1 1] End lerp timer on destroyed
  (Utilities|Time|ClearandInvalidateTimerbyHandle (Variables|Default|GetLerpTimer))
  (|CallParentFunction))

(event Custom|Event_StartFuse
  (params (DetonationTime :type Float%20%28double-precision%29))
  ;[1 1 1] Set event to explode when the passed in remaining fuse time expires
  (Utilities|FlowControl|DoOnce false
    (:Completed
      (Variables|Default|SetDetonationTime DetonationTime))))

(event EventOnResumeSimulation
  (|CallParentFunction)
  (Components|Movement|SetUpdatedComponent (Variables|Movement|GetProjectileMovementComponent) (Variables|Collision|GetCapsuleComponent))
  (Components|Movement|SetUpdatedComponent (Variables|Default|GetRotatingMovement) (Variables|Collision|GetCapsuleComponent))
  (Networking|ForceNetUpdate))

(event EventFuseEnded
  (bind _effect_distance (Variables|B_Prj_Grenade|GetEffect_Distance))
  ;[0.1009 0.03241 0.9948 0.5] Fuse Ended - explode
  (Variables|Default|SetbFuseEnded true)
  (Transformation|SetWorldRotation _effect_distance "90.000000,0.000000, 0" false false)
  (Components|Activation|Activate _effect_distance false)
  (Effects|Components|ParticleSystem|SetVectorParameter _effect_distance "Size" (Math|Conversions|ToVector(Float) (Variables|Default|GetExplosionRadius)))
  (if (not (Team|OnSameTeam (Game|GetInstigatorController) (Game|GetPlayerController 0)))
    (Effects|Components|ParticleSystem|SetVectorParameter _effect_distance "Color" "1.500000,0.110000,0.190000"))
  (Delivery|DoExplosion)
  (Components|Movement|SetUpdatedComponent (Variables|Movement|GetProjectileMovementComponent) (Variables|Collision|GetCapsuleComponent))
  (Components|Movement|SetUpdatedComponent (Variables|Default|GetRotatingMovement) (Variables|Collision|GetCapsuleComponent))
  (Networking|ForceNetUpdate)
  (bind _new_velocity (Default|CalculateDirection (Variables|Default|GetImpact%20Normal) true))
  (Variables|/Script/Engine.MovementComponent|SetVelocity _new_velocity (Variables|Movement|GetProjectileMovementComponent)))

(event EventOnExploded
  (params (HitActors :type Array%20of%20Actor%20Object%20References) (HitResults :type Array%20of%20Hit%20Result%20Structures))
  (bind _returnvalue (Math|Rotator|MakeRotator 0.0 90))
  (bind _returnvalue_2 (Game|GetInstigator))
  (bind _returnvalue_3 (Transformation|GetActorLocation))
  (bind _actual_radius (Variables|Default|GetActual%20Radius))
  (bind _returnvalue_4 (/ _actual_radius 5e+02))
  (bind _e_surface_type (Variables|Default|GetE_Surface_Type))
  (bind _returnvalue_5 (Math|Conversions|ToInteger(Byte) _e_surface_type))
  (bind _returnvalue_6 (Transformation|GetActorLocation))
  (bind _returnvalue_7 (Transformation|GetActorLocation))
  (bind _returnvalue_8 (Transformation|GetActorLocation))
  (bind _cue_explosionsound (Variables|Sounds|GetCue_ExplosionSound))
  (bind _cue_explosionsound_2 (Variables|Sounds|GetCue_ExplosionSound))
  (bind _cuebounceexplosionsound (Variables|Default|GetCueBounceExplosionSound))
  ;[1 0.8358 0.0634] OnExploded - Emitter, Sound, Camera Shake
  ; TODO: merge functionality in the sequence
  ; Top logic is TNT, Bottom is cherry smoke unstable grenade
  (Utilities|FlowControl|Sequence
    (:then_0
      (Utilities|Time|ClearandInvalidateTimerbyHandle (Variables|Default|GetFuseIndicatorTimer))
      (Utilities|Time|ClearandInvalidateTimerbyHandle (Variables|Default|GetLerpTimer))
      (Projectile|RemoveGameplayCue%28Looping%29Local self (Variables|Default|GetFuseCue) (Ability|GameplayCue|MakeGameplayCueParameters :NormalizedMagnitude 0.0 :RawMagnitude 0.0 :EffectContext "()" :AggregatedSourceTags "(GameplayTags=)" :AggregatedTargetTags "(GameplayTags=)" :Location "0,0,0" :Normal "0,0,0" :GameplayEffectLevel 1 :AbilityLevel 1 :bReplicateLocationWhenUsingMinimalRepProxy false))
      (Networking|CanExecuteCosmeticEvents
        (:True
          (bind _outhit (Collision|LineTraceByChannel (+ _returnvalue_7 "0.000000,0.000000,50.000000") (+ _returnvalue_7 "0.000000,0.000000,-50.000000") "TraceTypeQuery12"))
          (bind (_outhit_bblockinghit _outhit_binitialoverlap _outhit_time _outhit_distance _outhit_location _outhit_impactpoint _outhit_normal _outhit_impactnormal _outhit_physmat _outhit_hitactor _outhit_hitcomponent) (Collision|BreakHitResult _outhit))
          (if (Utilities|String|Contains (Utilities|GetDisplayName _outhit_hitcomponent) "waterbody")
            (bind _returnvalue_9 (Effects|Components|ParticleSystem|SpawnFXAtLocation (CallFunction|GetExplosion true) _returnvalue_7 _returnvalue))
            (Rendering|SetRenderCustomDepth _returnvalue_9 true)
            (Audio|PlaySoundatLocation (Variables|Sounds|GetCue_ExplosionSound) _returnvalue_8)
            (Audio|PlaySoundatLocation (Variables|Default|GetCue%20Water%20Sound) _returnvalue_8)
            (else
              (bind _returnvalue_10 (Effects|Components|ParticleSystem|SpawnFXAtLocation (CallFunction|GetExplosion false) _returnvalue_7 _returnvalue))
              (Rendering|SetRenderCustomDepth _returnvalue_10 true)
              (Components|Activation|Deactivate (Variables|Default|GetNS_Fuse_Particle))
              (if (Variables|Default|GetUse%20Shockwave%20Based%20Explosion)
                (Effects|Components|ParticleSystem|SetVectorParameter _returnvalue_10 "RadiusSize" (Math|Conversions|ToVector(Float) _returnvalue_4))
                (Effects|Components|ParticleSystem|SetFloatParameter _returnvalue_10 "SphereRadius" (* _returnvalue_4 1.5e+02))
                (Effects|Components|ParticleSystem|SetFloatParameter _returnvalue_10 "Global Scale" 0.95))
              (GameplayCue|ExecuteGameplayCue%28Burst%29Local self (Variables|Default|GetFeedbackCue) (Ability|GameplayCue|MakeGameplayCueParameters 0.0)) ; Cam shake + controller vibration
              (Class|BPSurfaceTypeSoundComponent|PlaySound (Variables|Default|GetBP_SurfaceTypeSoundComponent) (Transformation|GetActorLocation))
              (Audio|PlaySoundatLocation _cue_explosionsound _returnvalue_6) ; Play Explosion
            ))
          (Audio|Components|Audio|FadeOut (Variables|B_Prj_Grenade|GetGrenadeFuse_AudioComponent) 0.2 0.0 "Linear" :!class /Script/Engine.AudioComponent) ; Fade Out Fuse Sound
          (Fortnite|ShowSoundIndicatorLocal self _returnvalue_6 (Variables|/Script/Engine.SoundBase|GetMaxDistance _cue_explosionsound) "Gunshot"))))
    (:then_1
      (if (Variables|Default|GetbFuseEnded)
        ;[1 1 1] Bounce Explosions
        (Networking|SwitchHasAuthority
          (:Authority
            (GameplayCue|ExecuteGameplayCueOnActor(Burst) self (Variables|Default|GetGC_OnExplode) (Ability|GameplayCue|MakeGameplayCueParameters (select (Variables|Default|GetFinalBounce) 1.0 0.0)))
            (Ability|SendGameplayEventtoActor (Game|GetInstigator) "(TagName=\"Event.Teacake.Dynamite.Explosion\")" (Utilities|Struct|MakeGameplayEventData "(TagName=\"\")" self 0 0 0 "()" "(GameplayTags=)" "(GameplayTags=)" 0.0 (Ability|TargetData|AbilityTargetDatafromHitResult (Collision|MakeHitResult false false 0.0 0.0 (Transformation|GetActorLocation)))))
            (if (Variables|Default|GetFinalBounce)
              (Delivery|Kill)))
          (:Remote
            (bind _returnvalue_11 (Effects|Components|ParticleSystem|SpawnFXAtLocation (CallFunction|GetExplosion (or (Water|IsinWater (Variables|Water|GetWaterInteractionComponent)) (Utilities|Enum|Equal(Enum) (Variables|Default|GetE_Surface_Type) "SurfaceType8"))) (Transformation|GetActorLocation) (Variables|Default|GetExplosion%20Rotation)))
            ;;[1 1 1] PhysSurface Interaction for explosion Niagara FX
            (bind _checkedphysicalsurfaceid (CallFunction|PhysicalSurfaceCheck _returnvalue_5))
            (if _checkedphysicalsurfaceid
              (bind _returnvalue_12 (Effects|Components|ParticleSystem|SpawnFXAtLocation (Variables|VFX|GetFX_Physical%20Surface)))
              (Effects|Components|ParticleSystem|SetIntParameter _returnvalue_12 "PhysicalSurface" _returnvalue_5)
              (Effects|Components|ParticleSystem|SetFloatParameter _returnvalue_12 "Global Scale" 0.8))
            (Rendering|SetRenderCustomDepth _returnvalue_11 true) ; DISABLED
            (Utilities|FlowControl|Sequence
              (:then_0
                (Camera|PlayWorldCameraShake (Variables|Default|GetExplosionCameraShake) (Transformation|GetActorLocation) 200 800))
              (:then_1
                (if (Variables|Default|GetFinalBounce)
                  (Utilities|IsValid _cue_explosionsound_2
                    (:Is%20Valid
                      (Audio|PlaySoundatLocation _cue_explosionsound_2 _returnvalue_3 "0, 0, 0" 1.0 1.0 0.0 0 0 _returnvalue_2) ; Play Explosion
                      (Fortnite|ShowSoundIndicatorLocal :Location (Transformation|GetActorLocation) :MaxAudibleDistance (Variables|/Script/Engine.SoundBase|GetMaxDistance (Variables|Sounds|GetCue_ExplosionSound)) :IndicatorType "Gunshot" :Instigator (Variables|Actor|GetInstigator) :Tint "(R=0.000000,G=0.000000,B=0.000000,A=0.000000)" :SoundBase (Variables|Sounds|GetCue_ExplosionSound) :AttenuationMaxRangeMultiplier 1.0 :AttenuationIntensityMultiplier 1.0 :!label _label) ; Indicator for explosion
                    )
                    (:Is%20Not%20Valid
                      _))
                  (else
                    (Utilities|IsValid _cuebounceexplosionsound
                      (:Is%20Valid
                        (Audio|PlaySoundatLocation _cuebounceexplosionsound _returnvalue_3 "0, 0, 0" 1.0 1.0 0.0 0 0 _returnvalue_2) ; Play Explosion
                        (goto _label))
                      (:Is%20Not%20Valid
                        _))))
                (bind _waterfuseaudiocomponent (Variables|Default|GetWaterFuseAudioComponent)
                  (:then
                    (Audio|Components|Audio|FadeOut _waterfuseaudiocomponent 0.2 0.0 "Linear" :!class /Script/Engine.AudioComponent))
                  (:else)))
              (:then_2
                (Fortnite|PlayLocalForceFeedbackatLocationMulti (Variables|Default|GetExplosionForceFeedbackNear) (Variables|Default|GetExplosionForceFeedbackFar) (Transformation|GetActorLocation) 4.5e+02 7e+02 "Explosion"))
              (:then_3
                (if (Variables|Default|GetFinalBounce)
                  (Collision|SetActorEnableCollision false)
                  (Development|SetHiddeninGame (Variables|B_Prj_Grenade|GetMesh) true true)
                  (Audio|Components|Audio|FadeOut (Variables|B_Prj_Grenade|GetGrenadeFuse_AudioComponent) 0.3 0.0 "Linear" :!class /Script/Engine.AudioComponent)
                  (Utilities|Time|ClearandInvalidateTimerbyHandle (Variables|Default|GetFuseIndicatorTimer))
                  (Components|Activation|Deactivate (Variables|Default|GetNS_Fuse_Particle)))))))))))

(event Custom|Water_Bounce
  (bind _currentnumberofbounces (Variables|Default|GetCurrentNumberOfBounces))
  (bind _effect_distance (Variables|B_Prj_Grenade|GetEffect_Distance))
  ;[0 0.9688 0.1092] OnBounce - Do an explosion
  ; Wait until dynamite fuse is done, initial explosion happens, then have it bounce around detonating again a few times
  (Utilities|FlowControl|Sequence
    (:then_0
      ;;[0.9688 0.05422 0.8714 0.5] Start fuse on first bounce
      (if (> (ScalableFloat|GetCachedScalableFloatValue (Variables|Default|GetHF_GateNewLogic)) 0)
        _
        (elif (> (ScalableFloat|GetCachedScalableFloatValue (Variables|Default|GetHF_StartFuseOnFirstBounce)) 0)
          (CallFunction|EventStartFuse (+ (Math|Float|Clamp(Float) (ScalableFloat|GetCachedScalableFloatValue (Variables|Default|GetHF_FuseDuration)) (Variables|Default|GetMinFuseTime) (ScalableFloat|GetCachedScalableFloatValue (Variables|Default|GetHF_FuseDuration))) (Fortnite|Sync|Time|GetServerWorldTimeSeconds))))))
    (:then_1
      (Utilities|FlowControl|DoOnce false
        (:Completed
          (if (not (Variables|Default|GetbFuseEnded))
            (Variables|Default|SetStopLocation (.ImpactPoint (Collision|BreakHitResult (stale EventOnBounce Hit))))
            (Variables|Default|SetInititalBounceLocation (.Location (Collision|BreakHitResult (stale EventOnBounce Hit))))
            (bind _new_velocity (Default|CalculateDirection (.ImpactNormal (Collision|BreakHitResult (stale EventOnBounce Hit)))))
            (Utilities|FlowControl|Sequence
              (:then_0
                ;;[0 0.4622 0.9427 0.5] Store off normal and velocity to use after initial explode
                (Variables|Default|SetImpact%20Normal (.ImpactNormal (Collision|BreakHitResult (stale EventOnBounce Hit))))
                (Variables|Default|SetNew%20Velocity _new_velocity))
              (:then_1
                (bind _output_get (Variables|Default|SetE_Surface_Type (Variables|/Script/PhysicsCore.PhysicalMaterial|GetSurfaceType (.PhysMat (Collision|BreakHitResult (stale EventOnBounce Hit))))))
                (Class|BPSurfaceTypeSoundComponent|SetSurfaceType (Variables|Default|GetBP_SurfaceTypeSoundComponent) (Variables|/Script/PhysicsCore.PhysicalMaterial|GetSurfaceType (.PhysMat (Collision|BreakHitResult (stale EventOnBounce Hit)))) (.HitActor (Collision|BreakHitResult (stale EventOnBounce Hit))))
                (switch Utilities|FlowControl|Switch|SwitchonEPhysicalSurface _output_get
                  (:SurfaceType24
                    (CallFunction|ForceFuseonContactEventifStartFuseonStop)))))))))
    (:then_2
      (if (Variables|Default|GetbFuseEnded)
        (Variables|Default|SetE_Surface_Type "SurfaceType_Default")
        (bind _bouncelimitreached (Default|HasReachedBounceLimit))
        (if (not _bouncelimitreached)
          (Networking|SwitchHasAuthority
            (:Authority
              (Math|Integer|IncrementInt _currentnumberofbounces
                (:%20%20
                  (if (not (> _currentnumberofbounces 0))
                    (Variables|Default|SetInititalBounceLocation (.Location (Collision|BreakHitResult (stale EventOnBounce Hit)))))
                  (bind _new_velocity_2 (Default|CalculateDirection (.ImpactNormal (Collision|BreakHitResult (stale EventOnBounce Hit)))))
                  (Variables|/Script/Engine.MovementComponent|SetVelocity _new_velocity_2 (Variables|Movement|GetProjectileMovementComponent)))))
            (:Remote
              _)))))
    (:then_3
      ;;[0.9583 0.9517 0.9319 0.5] Bounce cue
      (Networking|SwitchHasAuthority
        (:Remote
          (Utilities|IsValid (Variables|Sounds|GetCue_Bounce)
            (:Is%20Valid
              (Audio|OnCollision (Variables|Default|GetFortCollisionAudio) (.ImpactNormal (Collision|BreakHitResult (stale EventOnBounce Hit))) (* (.ImpactPoint (Collision|BreakHitResult (stale EventOnBounce Hit))) (Variables|/Script/Engine.MovementComponent|GetVelocity (Variables|Movement|GetProjectileMovementComponent))))))))))
  (Transformation|SetWorldRotation _effect_distance "0, 0, 0" false false)
  (Components|Activation|Activate _effect_distance false)
  (Effects|Components|ParticleSystem|SetVectorParameter _effect_distance "Size" (Math|Conversions|ToVector(Float) (Variables|Default|GetExplosionRadius)))
  (if (not (Team|OnSameTeam (Game|GetInstigatorController) (Game|GetPlayerController 0)))
    (Effects|Components|ParticleSystem|SetVectorParameter _effect_distance "Color" "1.500000,0.110000,0.190000"))
  (Delivery|DoExplosion)
  (Development|PrintString "bounce"))

(event EventOnBounce
  (params (Hit :type Hit%20Result%20Structure%20%28by%20ref%29))
  (bind (_bblockinghit _binitialoverlap _time _distance _location _impactpoint _normal _impactnormal _physmat _hitactor) (Collision|BreakHitResult Hit))
  (bind (_bblockinghit_2 _binitialoverlap_2 _time_2 _distance_2 _location_2 _impactpoint_2 _normal_2 _impactnormal_2) (Collision|BreakHitResult Hit))
  (bind _currentnumberofbounces (Variables|Default|GetCurrentNumberOfBounces))
  (bind _surfacetype (Variables|/Script/PhysicsCore.PhysicalMaterial|GetSurfaceType _physmat))
  (bind _effect_distance (Variables|B_Prj_Grenade|GetEffect_Distance))
  ;[0 0.9688 0.1092] OnBounce - Do an explosion
  ; Wait until dynamite fuse is done, initial explosion happens, then have it bounce around detonating again a few times
  (Utilities|FlowControl|Sequence
    (:then_0
      ;;[0.9688 0.05422 0.8714 0.5] Start fuse on first bounce
      (if (> (ScalableFloat|GetCachedScalableFloatValue (Variables|Default|GetHF_GateNewLogic)) 0)
        _
        (elif (> (ScalableFloat|GetCachedScalableFloatValue (Variables|Default|GetHF_StartFuseOnFirstBounce)) 0)
          (CallFunction|EventStartFuse (+ (Math|Float|Clamp(Float) (ScalableFloat|GetCachedScalableFloatValue (Variables|Default|GetHF_FuseDuration)) (Variables|Default|GetMinFuseTime) (ScalableFloat|GetCachedScalableFloatValue (Variables|Default|GetHF_FuseDuration))) (Fortnite|Sync|Time|GetServerWorldTimeSeconds))))))
    (:then_1
      (Utilities|FlowControl|DoOnce false
        (:Completed
          (if (not (Variables|Default|GetbFuseEnded))
            (Variables|Default|SetStopLocation _impactpoint)
            (Variables|Default|SetInititalBounceLocation _location)
            (bind _new_velocity (Default|CalculateDirection _impactnormal))
            (Utilities|FlowControl|Sequence
              (:then_0
                ;;[0 0.4622 0.9427 0.5] Store off normal and velocity to use after initial explode
                (Variables|Default|SetImpact%20Normal _impactnormal)
                (Variables|Default|SetNew%20Velocity _new_velocity))
              (:then_1
                (bind _output_get (Variables|Default|SetE_Surface_Type _surfacetype))
                (Class|BPSurfaceTypeSoundComponent|SetSurfaceType (Variables|Default|GetBP_SurfaceTypeSoundComponent) _surfacetype _hitactor)
                (switch Utilities|FlowControl|Switch|SwitchonEPhysicalSurface _output_get
                  (:SurfaceType24
                    (CallFunction|ForceFuseonContactEventifStartFuseonStop)))))))))
    (:then_2
      (if (Variables|Default|GetbFuseEnded)
        (Variables|Default|SetE_Surface_Type "SurfaceType_Default")
        (bind _bouncelimitreached (Default|HasReachedBounceLimit))
        (if (not _bouncelimitreached)
          (Networking|SwitchHasAuthority
            (:Authority
              (Math|Integer|IncrementInt _currentnumberofbounces
                (:%20%20
                  (if (not (> _currentnumberofbounces 0))
                    (Variables|Default|SetInititalBounceLocation _location_2))
                  (bind _new_velocity_2 (Default|CalculateDirection _impactnormal_2))
                  (Variables|/Script/Engine.MovementComponent|SetVelocity _new_velocity_2 (Variables|Movement|GetProjectileMovementComponent)))))
            (:Remote
              _)))))
    (:then_3
      ;;[0.9583 0.9517 0.9319 0.5] Bounce cue
      (Networking|SwitchHasAuthority
        (:Remote
          (Utilities|IsValid (Variables|Sounds|GetCue_Bounce)
            (:Is%20Valid
              (Audio|OnCollision (Variables|Default|GetFortCollisionAudio) _impactnormal_2 (* _impactpoint_2 (Variables|/Script/Engine.MovementComponent|GetVelocity (Variables|Movement|GetProjectileMovementComponent))))))))))
  (Transformation|SetWorldRotation _effect_distance "0, 0, 0" false false)
  (Components|Activation|Activate _effect_distance false)
  (Effects|Components|ParticleSystem|SetVectorParameter _effect_distance "Size" (Math|Conversions|ToVector(Float) (Variables|Default|GetExplosionRadius)))
  (if (not (Team|OnSameTeam (Game|GetInstigatorController) (Game|GetPlayerController 0)))
    (Effects|Components|ParticleSystem|SetVectorParameter _effect_distance "Color" "1.500000,0.110000,0.190000"))
  (Delivery|DoExplosion)
  (Development|PrintString "bounce"))

(event OnEnterWaterEvent(WaterInteractionComponent)
  (params (WaterBody :type Fort%20Water%20Body%20Actor%20Object%20Reference) (WaterInteractionComponent :type Fort%20Water%20Interaction%20Component%20Object%20Reference) (bIsFirstBody :type Boolean))
  ;[0.8698 0.08154 0.2343 0.5] Hit Water Surface reaction.  Added this because the bounce doesn’t occur in the parent BP.
  (CallFunction|HeldWaterImpact)
  (Utilities|FlowControl|Sequence
    (:then_0
      (Utilities|FlowControl|DoOnce false
        (:Completed
          (CallFunction|WaterSurfaceHitFX)
          (CallFunction|ForceFuseonContactEventifStartFuseonStop))))
    (:then_1
      (CallFunction|WaterBounce))))

(event Custom|WaterSurfaceHitFX
  ;[0.8698 0.08154 0.2343 0.5] Hit Water Surface reaction.  Added this because the bounce doesn’t occur in the parent BP.
  (Effects|Components|ParticleSystem|SpawnFXAtLocation (Variables|VFX|GetNS_HitWater) (Transformation|GetActorLocation) (Math|Rotator|MakeRotator 0.0 90)))

(event Custom|ForceFuseOnContactEventIfStartFuseOnStop
  ;[0.974 0.974 0.974 0.5] If we want to start fuse on 'landed' we need to force it here, as the prj won't stop on water or ice
  (if (> (ScalableFloat|GetCachedScalableFloatValue (Variables|Default|GetHF_StartFuseOnLanded)) 0)
    (CallFunction|EventStartFuse (+ (Math|Float|Clamp(Float) (ScalableFloat|GetCachedScalableFloatValue (Variables|Default|GetHF_FuseDuration)) (Variables|Default|GetMinFuseTime) (ScalableFloat|GetCachedScalableFloatValue (Variables|Default|GetHF_FuseDuration))) (Fortnite|Sync|Time|GetServerWorldTimeSeconds)))))

(fn CalculateDirection
  (params (HitNormal :type Vector) (UseStoredVelocity :type Boolean))
  (bind _projectilemovementcomponent (Variables|Movement|GetProjectileMovementComponent))
  (Variables|/Script/Engine.ProjectileMovementComponent|SetProjectileGravityScale 0.7 _projectilemovementcomponent)
  (bind _output_get (LocalVariables|SetCurrentDirection (Math|Vector|Normalize2D(Vector) (Math|Vector|SelectVector (Variables|Default|GetNew%20Velocity) (Variables|/Script/Engine.MovementComponent|GetVelocity _projectilemovementcomponent) UseStoredVelocity))))
  (Utilities|FlowControl|Sequence
    (:then_0
      (if (> (Variables|Default|GetCurrentNumberOfBounces) 0)
        (bind _output_get_2 (LocalVariables|SetTargetBounceLocation (+ (Variables|Default|GetInititalBounceLocation) (* (Math|Random|RandomUnitVector) (ScalableFloat|GetCachedScalableFloatValue (Variables|Default|GetHF_BouncesRadius))))))
        (LocalVariables|SetNewVelocity (Math|Vector|MakeVector (.x (* (Math|Vector|Normalize2D(Vector) (- _output_get_2 (Transformation|GetActorLocation))) (ScalableFloat|GetCachedScalableFloatValue (Variables|Default|GetHF_BouncesForce)))) (.y (* (Math|Vector|Normalize2D(Vector) (- _output_get_2 (Transformation|GetActorLocation))) (ScalableFloat|GetCachedScalableFloatValue (Variables|Default|GetHF_BouncesForce)))) (ScalableFloat|GetCachedScalableFloatValue (Variables|Default|GetHF_BouncesHeight))))
        (else
          (LocalVariables|SetNewVelocity (Math|Vector|MakeVector (.x (* (Math|Vector|RotateVectorAroundAxis _output_get (Math|Random|RandomFloatinRange (* (- 1.8e+02 (Math|Float|Absolute(Float) (ScalableFloat|GetCachedScalableFloatValue (Variables|Default|GetHF_FirstBounceDeadzoneAngle)))) -1.0) (- 1.8e+02 (Math|Float|Absolute(Float) (ScalableFloat|GetCachedScalableFloatValue (Variables|Default|GetHF_FirstBounceDeadzoneAngle))))) HitNormal) (ScalableFloat|GetCachedScalableFloatValue (Variables|Default|GetHF_BouncesForce)))) (.y (* (Math|Vector|RotateVectorAroundAxis _output_get (Math|Random|RandomFloatinRange (* (- 1.8e+02 (Math|Float|Absolute(Float) (ScalableFloat|GetCachedScalableFloatValue (Variables|Default|GetHF_FirstBounceDeadzoneAngle)))) -1.0) (- 1.8e+02 (Math|Float|Absolute(Float) (ScalableFloat|GetCachedScalableFloatValue (Variables|Default|GetHF_FirstBounceDeadzoneAngle))))) HitNormal) (ScalableFloat|GetCachedScalableFloatValue (Variables|Default|GetHF_BouncesForce)))) (ScalableFloat|GetCachedScalableFloatValue (Variables|Default|GetHF_BouncesHeight)))))))
    (:then_1
      (return (LocalVariables|GetNewVelocity)))))

(fn CalculateRotation
  (params (NewParam_X :type Float%20%28double-precision%29) (NewParam_Y :type Float%20%28double-precision%29) (NewParam_Z :type Float%20%28double-precision%29) (NewParam_X :type Float%20%28double-precision%29) (NewParam_Y :type Float%20%28double-precision%29) (NewParam_Z :type Float%20%28double-precision%29))
  (bind (_returnvalue_x _returnvalue_y _returnvalue_z) (Math|Vector|BreakVector (Math|Vector|Normalize (Math|Vector|MakeVector NewParam_X NewParam_Y) 0.0001)))
  (return (Math|Rotator|MakeRotator _returnvalue_x _returnvalue_y _returnvalue_z)))

(fn DrawDebugDirectionalArrow
  (params (End%20of%20Line :type Vector) (Debug%20Color%20Line :type Linear%20Color%20Structure :default "(R=1.000000,G=0.000000,B=0.067432,A=1.000000)"))
  (if (and (Variables|Default|GetDrawDebug) (Editor|InEditororPIE))
    (Rendering|Debug|DrawDebugArrow (Transformation|GetActorLocation) (+ (Transformation|GetActorLocation) (* End%20of%20Line 4e+02)) 0.0 Debug%20Color%20Line 1e+01 5.0)))

(fn HasReachedBounceLimit
  (Networking|SwitchHasAuthority
    (:Authority
      (Variables|Default|SetFinalBounce (> (Variables|Default|GetCurrentNumberOfBounces) (- (Variables|Default|GetNumberOfBouncesTillExplode) 1))))
    (:Remote
      _))
  (return (Variables|Default|GetFinalBounce)))

(fn MakeOrthogonalZ
  (params (InRotation :type Rotator) (Normal :type Vector))
  (bind _returnvalue (Math|Vector|DotProduct (.Z (Math|Rotator|GetAxes InRotation)) Normal))
  (bind _returnvalue_2 (* Normal _returnvalue))
  (bind _returnvalue_3 (- (.Z (Math|Rotator|GetAxes InRotation)) _returnvalue_2))
  (bind _returnvalue_4 (Math|Vector|Normalize _returnvalue_3))
  (bind _returnvalue_5 (Math|Vector|DotProduct (.Y (Math|Rotator|GetAxes InRotation)) _returnvalue_4))
  (bind (_x _y _z) (Math|Rotator|GetAxes InRotation))
  (bind _returnvalue_6 (Math|Vector|DotProduct _x _returnvalue_4))
  (if (> (Math|Float|Absolute(Float) _returnvalue_6) (Math|Float|Absolute(Float) _returnvalue_5))
    (return (Math|Rotator|MakeRotfromZX _returnvalue_4 (Math|Vector|Normalize (- _x (* _returnvalue_4 _returnvalue_6)))))
    (else
      (return (Math|Rotator|MakeRotfromZY _returnvalue_4 (Math|Vector|Normalize (- _y (* _returnvalue_4 _returnvalue_5))))))))

(fn OnRep_DetonationTime
  (Variables|Default|SetbFuseStarted true)
  (bind _output_get (Variables|Default|SetFuseTime (- (Variables|Default|GetDetonationTime) (Fortnite|Sync|Time|GetServerWorldTimeSeconds))))
  (Projectile|AddGameplayCue%28Looping%29Local self (Variables|Default|GetFuseCue) (Ability|GameplayCue|MakeGameplayCueParameters 0.0 _output_get))
  (bind _returnvalue (Utilities|Time|SetTimerbyEvent (EventDispatchers|CreateEvent :!event FuseEnded) _output_get))
  (Variables|Default|SetFuseTimer _returnvalue))

(fn OnRep_FinalBounce)

(fn OnRep_ResumeMovement
  (if (Variables|Default|GetResumeMovement)
    (Projectile|ResumeSimulation)
    (Components|Movement|SetUpdatedComponent (Variables|Movement|GetProjectileMovementComponent) (Variables|Collision|GetCapsuleComponent))
    (Components|Movement|SetUpdatedComponent (Variables|Default|GetRotatingMovement) (Variables|Collision|GetCapsuleComponent))
    (Networking|ForceNetUpdate)
    (bind _new_velocity (Default|CalculateDirection (Variables|Default|GetImpact%20Normal) true))
    (Variables|/Script/Engine.MovementComponent|SetVelocity (* _new_velocity 5.0) (Variables|Movement|GetProjectileMovementComponent))))

(fn SplinePointLocations
  (return (Utilities|Array|MakeArray (Math|Transform|InverseTransformLocation (Transformation|GetWorldTransform (Variables|Default|GetSpline)) (Transformation|GetSocketLocation (Variables|Default|GetTNTBundle) "FX_FuseBegin")) (Math|Transform|InverseTransformLocation (Transformation|GetWorldTransform (Variables|Default|GetSpline)) (Transformation|GetSocketLocation (Variables|Default|GetTNTBundle) "dyn_fuse2")) (Math|Transform|InverseTransformLocation (Transformation|GetWorldTransform (Variables|Default|GetSpline)) (Transformation|GetSocketLocation (Variables|Default|GetTNTBundle) "dyn_fuse1")) (Math|Transform|InverseTransformLocation (Transformation|GetWorldTransform (Variables|Default|GetSpline)) (Transformation|GetSocketLocation (Variables|Default|GetTNTBundle) "FX_FuseEnd")))))

(fn ConstructionScript
  (|CallParentFunction))
